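; Weapon Blueprint for the Umbramolt Kipper, a Fortnite weapon with a dash charge state and strike spark effects.
; On BeginPlay it stores the instigator as player pawn and binds a damage delegate whose handler is empty; equip events track the equipped state.
; SetDashChargeEnabled stores the charging flag and broadcasts it through a dispatcher.
; PlayStrikeSparks randomly fires a left or right GameplayCue burst on the current target, with debug prints.
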(event EventBeginPlay
  ;[0.7899 0.1002 0.2833] BeginPlay
  (Utilities|FlowControl|Sequence
    (:then_0
      (bind _asfort_player_pawn (Utilities|Casting|CastToFortPlayerPawn (Game|GetInstigator)))
      (Variables|Default|SetPlayerPawn _asfort_player_pawn)
      (EventDispatchers|BindEventtoOnDidDamage (Variables|Default|GetPlayerPawn) (delegate Custom|CustomEvent)))))

(event Custom|CustomEvent
  (params (DamagedActor :type Actor%20Object%20Reference) (Damage :type Float%20%28single-precision%29) (InstigatedBy :type Controller%20Object%20Reference) (DamageCauser :type Actor%20Object%20Reference) (HitLocation :type Vector) (FHitComponent :type Primitive%20Component%20Object%20Reference) (BoneName :type Name) (Momentum :type Vector)))

(event Custom|SetDashChargeEnabled
  (params (bIsDashChargeEnabled :type Boolean))
  ;[0.7899 0.1002 0.2833] BeginPlay
  ;[0.2317 0.8396 1] Dash Charge
  (bind _output_get (Variables|Charge|SetbIsChargingDash bIsDashChargeEnabled))
  (Default|CallCallonDashChargeEnabledorDisabled _output_get))

(event FortWeapon|EventOnEquippedAndAttached
  ;[0.7899 0.1002 0.2833] On equip/unequip
  (Variables|Default|SetCurrentlyEquipped true))

(event FortWeapon|EventOnUnEquippedAndDetached
  ;[0.7899 0.1002 0.2833] On equip/unequip
  (Variables|Default|SetCurrentlyEquipped false))

(event Custom|PlayStrikeSparks
  (bind _returnvalue (Ability|GameplayCue|MakeGameplayCueParameters 0.0))
  (bind _currenttarget (Variables|Default|GetCurrentTarget))
  ;[0.104 0.15 0.02087 0.5] Show 'blood' sparks on impact
  (if (Math|Random|RandomBool)
    (GameplayCue|ExecuteGameplayCueOnActor(Burst) _currenttarget (Variables|Default|GetGCStrikeSparksLeftSide) _returnvalue)
    (Development|PrintString "Struck left")
    (else
      (GameplayCue|ExecuteGameplayCueOnActor(Burst) _currenttarget (Variables|Default|GetGCStrikeSparksRightSide) _returnvalue)
      (Development|PrintString "Struck right"))))

(fn CreateDynamicMaterials)

(fn OnMoveModeChanged
  (params (Character :type Character%20Object%20Reference) (PrevMovementMode :type EMovementMode%20Enum :default "MOVE_None") (PreviousCustomMode :type Byte)))

(fn ConstructionScript
  (|CallParentFunction))
